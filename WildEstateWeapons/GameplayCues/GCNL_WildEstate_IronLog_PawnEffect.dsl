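; Looping gameplay cue for the Iron Log weapon's effect on an affected pawn, likely when hit.
; On loop start, it stores the first particle component and sets its sort offset, duration parameter and Niagara impact position from the cue location.
; It also attaches a separate head Niagara system to the parent's FX_Head socket.
; Removal deactivates only the stored component; OnBurst and OnExecute retain debug prints.
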(event EventOnLoopingStart(EitherVFX-preferthisversion)
  (params (MyTarget :type Actor%20Object%20Reference) (Parameters :type Gameplay%20Cue%20Parameters%20Structure%20%28by%20ref%29) (ParticleComponents :type Array%20of%20FXSystem%20Component%20Object%20References) (AudioComponents :type Array%20of%20Audio%20Component%20Object%20References))
  (bind (_normalizedmagnitude _rawmagnitude _effectcontext _matchedtagname _originaltag _aggregatedsourcetags _aggregatedtargettags _location) (Ability|GameplayCue|BreakGameplayCueParameters Parameters))
  (Utilities|FlowControl|Sequence
    (:then_0
      ;[0.04115 0.04287 0.4115] Set parameters on primary fx component
      (bind _output_get (Variables|Default|SetPrimaryFxComponent (Utilities|Array|Get(acopy) ParticleComponents)))
      (Utilities|IsValid _output_get
        (:Is%20Valid
          (Rendering|SetTranslucencySortDistanceOffset _output_get -2.5e+02)
          (Effects|Components|ParticleSystem|SetFloatParameter _output_get (Variables|Default|GetFXDurationName) (ScalableFloat|GetValueatLevel (Variables|Default|GetFXDuration)))
          (bind _asniagara_particle_system_component (Utilities|Casting|CastToNiagaraComponent _output_get))
          (Niagara|SetNiagaraVariable(Position) _asniagara_particle_system_component "ImpactPosition" _location)
          (Niagara|SpawnSystemAttached "/WildEstateWeapons/Gameplay/IronLog/FX/Impacted/FNE_Impacts/NS_IronLogHitPlayerMakeupFX_Loop_FNE_head.NS_IronLogHitPlayerMakeupFX_Loop_FNE_head" (Transformation|GetAttachParent _asniagara_particle_system_component) "FX_Head" "0, 0, 0" "0, 0, 0" "SnapToTarget" true))))
    (:then_1
      (|CallParentFunction MyTarget Parameters ParticleComponents AudioComponents))))

(event EventOnRemoval(EitherVFX-preferthisversion)
  (params (MyTarget :type Actor%20Object%20Reference) (Parameters :type Gameplay%20Cue%20Parameters%20Structure%20%28by%20ref%29) (ParticleComponents :type Array%20of%20FXSystem%20Component%20Object%20References) (AudioComponents :type Array%20of%20Audio%20Component%20Object%20References) (BurstCameraShakeInstance :type Camera%20Shake%20Base%20Object%20Reference) (BurstDecalInstance :type Decal%20Actor%20Object%20Reference))
  (|CallParentFunction MyTarget Parameters ParticleComponents AudioComponents BurstCameraShakeInstance BurstDecalInstance)
  (bind _primaryfxcomponent (Variables|Default|GetPrimaryFxComponent)
    (:then
      (Components|Activation|Deactivate _primaryfxcomponent))
    (:else)))

(event EventOnApplication(EitherVFX-preferthisone)
  (params (MyTarget :type Actor%20Object%20Reference) (Parameters :type Gameplay%20Cue%20Parameters%20Structure%20%28by%20ref%29) (ParticleComponents :type Array%20of%20FXSystem%20Component%20Object%20References) (AudioComponents :type Array%20of%20Audio%20Component%20Object%20References) (BurstCameraShakeInstance :type Camera%20Shake%20Base%20Object%20Reference) (BurstDecalInstance :type Decal%20Actor%20Object%20Reference))
  (|CallParentFunction MyTarget Parameters ParticleComponents AudioComponents BurstCameraShakeInstance BurstDecalInstance))

(fn OnBurst
  (params (MyTarget :type Actor%20Object%20Reference) (Parameters :type Gameplay%20Cue%20Parameters%20Structure%20%28by%20ref%29))
  (bind _returnvalue (|CallParentFunction MyTarget Parameters))
  (Development|PrintString 0 "Hello" true true "(R=1.000000,G=0.000000,B=0.000446,A=1.000000)")
  (return _returnvalue))

(fn OnExecute
  (params (MyTarget :type Actor%20Object%20Reference) (Parameters :type Gameplay%20Cue%20Parameters%20Structure%20%28by%20ref%29))
  (bind _returnvalue (|CallParentFunction MyTarget Parameters))
  (Development|PrintString 0 "ON Execute" true true "(R=1.000000,G=0.000000,B=0.000446,A=1.000000)")
  (return _returnvalue))

(fn ConstructionScript)
